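; Looping gameplay cue for the Clyde Athena small health speed effect, with a per-character fix for VFX attachment.
; After the parent looping-start logic, it takes the first particle component and checks the player's cosmetic character and Body mesh.
; For Character_CitrusSpoon with an FX_Head socket, it snaps the effect to that socket and binds the Niagara mesh parameter to Body.
; Otherwise it only stores the default character mesh; both branches print debug text.
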(event EventOnLoopingStart(EitherVFX-preferthisversion)
  (params (MyTarget :type Actor%20Object%20Reference) (Parameters :type Gameplay%20Cue%20Parameters%20Structure%20%28by%20ref%29) (ParticleComponents :type Array%20of%20FXSystem%20Component%20Object%20References) (AudioComponents :type Array%20of%20Audio%20Component%20Object%20References))
  (bind _output (Utilities|Array|Get(acopy) ParticleComponents))
  (|CallParentFunction MyTarget Parameters ParticleComponents AudioComponents)
  ;[0 0 0] Fix per character
  (bind _asfort_player_pawn (Utilities|Casting|CastToFortPlayerPawn MyTarget))
  (bind _asniagara_particle_system_component (Utilities|Casting|CastToNiagaraComponent _output))
  (bind (_cosmeticloadout_bannericonid _cosmeticloadout_bannercolorid _cosmeticloadout_skydivecontrail _cosmeticloadout_glider _cosmeticloadout_pickaxe _cosmeticloadout_character) (Utilities|Struct|BreakFortAthenaLoadout (Variables|/Script/FortniteGame.FortPlayerPawn|GetCosmeticLoadout_BannerIconId _asfort_player_pawn)))
  (if (and (Transformation|DoesSocketExist (Customization|GetSkeletalMeshforPartType _asfort_player_pawn "Body" :!class /Script/FortniteGame.FortPlayerPawn) (Utilities|Name|MakeLiteralName "FX_Head")) (Utilities|String|EqualExactly(String) (Utilities|GetDisplayName _cosmeticloadout_character) "Character_CitrusSpoon"))
    (Variables|Default|SetMeshAttatchComponent (Customization|GetSkeletalMeshforPartType _asfort_player_pawn "Body" :!class /Script/FortniteGame.FortPlayerPawn))
    (Development|PrintString 0 "Attatching to body")
    (Transformation|AttachComponentToComponent _output (Variables|Default|GetMeshAttatchComponent) "FX_Head" "SnapToTarget" "SnapToTarget" "KeepRelative" true)
    (Niagara|SetNiagaraSkeletalMeshComponent _asniagara_particle_system_component "SK_Mesh" (Variables|Default|GetMeshAttatchComponent))
    (else
      (Variables|Default|SetMeshAttatchComponent (Variables|/Script/Engine.Character|GetMesh _asfort_player_pawn)) ; all else fails go to the OG
      (Development|PrintString 0 "attatching to Default SK Mesh"))))

(fn ConstructionScript)
