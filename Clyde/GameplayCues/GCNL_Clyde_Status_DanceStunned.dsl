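; Looping gameplay cue notify for the Clyde dance-stunned status, covering animation and particle effects on affected AI and players.
; On application it sets AI pawns to the shared dance anim state, and on removal returns them to idle.
; For players, removal stops the Dance Grenade emote montage, which this cue never starts itself.
; Loop start and player removal set the first particle component's LocalClient parameter from whether the player is locally controlled.
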(event EventOnLoopingStart(EitherVFX-preferthisversion)
  (params (MyTarget :type Actor%20Object%20Reference) (Parameters :type Gameplay%20Cue%20Parameters%20Structure%20%28by%20ref%29) (ParticleComponents :type Array%20of%20FXSystem%20Component%20Object%20References) (AudioComponents :type Array%20of%20Audio%20Component%20Object%20References))
  (bind _asplayer_pawn_athena (Utilities|Casting|CastToPlayerPawn_Athena MyTarget))
  (Effects|Components|ParticleSystem|SetBoolParameter (Utilities|Array|Get(acopy) ParticleComponents) "LocalClient" (Pawn|IsLocallyControlled _asplayer_pawn_athena)))

(event EventOnApplication(EitherVFX-preferthisone)
  (params (MyTarget :type Actor%20Object%20Reference) (Parameters :type Gameplay%20Cue%20Parameters%20Structure%20%28by%20ref%29) (ParticleComponents :type Array%20of%20FXSystem%20Component%20Object%20References) (AudioComponents :type Array%20of%20Audio%20Component%20Object%20References) (BurstCameraShakeInstance :type Camera%20Shake%20Base%20Object%20Reference) (BurstDecalInstance :type Decal%20Actor%20Object%20Reference))
  ;[1 1 1] Handle AI
  (bind _asfort_aipawn (Utilities|Casting|CastToFortAIPawn MyTarget))
  (Animation|SetSharedAnimState (Variables|/Script/FortniteGame.FortAIPawn|GetAIPawnAbilitySystemComponent _asfort_aipawn) "Anim_Dance"))

(event EventOnRemoval(EitherVFX-preferthisversion)
  (params (MyTarget :type Actor%20Object%20Reference) (Parameters :type Gameplay%20Cue%20Parameters%20Structure%20%28by%20ref%29) (ParticleComponents :type Array%20of%20FXSystem%20Component%20Object%20References) (AudioComponents :type Array%20of%20Audio%20Component%20Object%20References) (BurstCameraShakeInstance :type Camera%20Shake%20Base%20Object%20Reference) (BurstDecalInstance :type Decal%20Actor%20Object%20Reference))
  ;[1 1 1] Handle AI
  (bind _asfort_aipawn (Utilities|Casting|CastToFortAIPawn MyTarget)
    (:then
      (Animation|SetSharedAnimState (Variables|/Script/FortniteGame.FortAIPawn|GetAIPawnAbilitySystemComponent _asfort_aipawn) "Anim_Idle"))
    (:CastFailed
      (bind _asplayer_pawn_athena (Utilities|Casting|CastToPlayerPawn_Athena MyTarget))
      (Animation|StopAnimMontage _asplayer_pawn_athena "/Clyde/Weapons/DanceGrenade/Emote_SmoothDrive_DanceGrenade_CMM_M.Emote_SmoothDrive_DanceGrenade_CMM_M")
      (Effects|Components|ParticleSystem|SetBoolParameter (Utilities|Array|Get(acopy) ParticleComponents) "LocalClient" (Pawn|IsLocallyControlled _asplayer_pawn_athena)))))

(fn ConstructionScript)
