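; Looping gameplay cue that dresses a player pawn in Freeze Trap ice-feet effects and sliding audio.
; On start it attaches emitters to both feet and binds footstep, hit, landing and movement-mode sounds; Tick drives slide audio from walking speed.
; For the local player, timers start a repeating pre-expiry sound that raises the emitters' Crack parameter and spawns Niagara crack bursts.
; Removal clears timers, spawns expire FX and unbinds events.
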(event Custom|BeginExpireTell
  ;[1 1 1] Handle ability expire warning - local player only
  (bind _returnvalue (Utilities|Time|SetTimerbyEvent (delegate Custom|ExpireWarningTick) (Variables|Default|GetExpirationSoundPeriod) true))
  (Variables|Default|SetExpirationSoundTimer _returnvalue)
  (Variables|Default|SetCrack 0.0))

(event Custom|ExpireWarningTick
  ;[1 1 1] Handle ability expire warning - local player only
  (Audio|SpawnSoundAttached "/Game/Sounds/Fort_Traps/Ice_Trap/IceTrap_PreExpireTell_Cue.IceTrap_PreExpireTell_Cue" (Variables|/Script/Engine.Actor|GetRootComponent (Variables|Default|GetTargetPlayer)))
  (bind _output_get (Variables|Default|SetCrack (+ (Variables|Default|GetCrack) (/ (- (Variables|Default|GetExpirationTellDuration) (Variables|Default|GetExpirationSoundPeriod)) (Variables|Default|GetExpirationTellDuration)))))
  (for _array_element (Variables|Default|GetSpawnedEmitters)
    (Effects|Components|ParticleSystem|SetFloatParameter _array_element "Crack" _output_get)
    (Effects|Components|ParticleSystem|SetFloatParameter _array_element "Time" (Utilities|Time|GetGameTimeinSeconds 0))
    (Niagara|SpawnSystematLocation 0 "/FNGameplayCues/Athena/Items/Traps/FreezeTrap/FX/NS_Ice_Foot_Crack.NS_Ice_Foot_Crack" (Transformation|GetWorldLocation _array_element) "0.000000,90.000000,0.000000")))

(event Custom|ClearExpireTell
  ;[1 1 1] Handle ability expire warning - local player only
  (Utilities|Time|ClearandInvalidateTimerbyHandle 0 (Variables|Default|GetExpireTellDelayTimer))
  (Utilities|Time|ClearandInvalidateTimerbyHandle 0 (Variables|Default|GetExpirationSoundTimer)))

(event EventOnLoopingStart(EitherVFX-preferthisversion)
  (params (MyTarget :type Actor%20Object%20Reference) (Parameters :type Gameplay%20Cue%20Parameters%20Structure%20%28by%20ref%29) (ParticleComponents :type Array%20of%20FXSystem%20Component%20Object%20References) (AudioComponents :type Array%20of%20Audio%20Component%20Object%20References))
  (bind _spawnedemitters (Variables|Default|GetSpawnedEmitters))
  (bind _output (Utilities|Array|Get(acopy) _spawnedemitters 1))
  (bind _output_2 (Utilities|Array|Get(acopy) _spawnedemitters))
  (Utilities|FlowControl|Sequence
    (:then_0
      (Variables|Default|SetSpawnedEmitters ParticleComponents)
      (if (Utilities|Array|IsValidIndex AudioComponents)
        (Variables|Default|SetSpawnedAudioComps AudioComponents)
        (else
          _))
      (CallFunction|ClearExpireTell)
      (Variables|Default|SetCrack 0.0))
    (:then_1
      (bind _asfort_player_pawn (Utilities|Casting|CastToFortPlayerPawn MyTarget))
      (Variables|Default|SetTargetPlayer _asfort_player_pawn)
      (Transformation|AttachComponentToComponent _output_2 (Variables|/Script/Engine.Character|GetMesh _asfort_player_pawn) "foot_l" "SnapToTarget" "SnapToTarget" "KeepWorld" true)
      (Effects|Components|ParticleSystem|SetVectorParameter _output_2 "BlockOrientation" "0.000000,0.000000,1.000000")
      (Transformation|AttachComponentToComponent _output (Variables|/Script/Engine.Character|GetMesh _asfort_player_pawn) "foot_r" "SnapToTarget" "SnapToTarget" "KeepWorld" true)
      (Effects|Components|ParticleSystem|SetVectorParameter _output "BlockOrientation" "0.000000,0.000000,-1.000000")
      (Effects|Components|ParticleSystem|SetFloatParameter (Variables|Default|GetSpawnedEmitters) "Time" (Utilities|Time|GetGameTimeinSeconds 0))
      (Utilities|FlowControl|Sequence
        (:then_0
          ;[1 1 1] Handle ability expire warning - local player only
          (if (Pawn|IsLocallyControlled (Variables|Default|GetTargetPlayer))
            (Variables|Default|SetDuration (ScalableFloat|GetValueatLevel (Variables|Default|GetDataTellDuration)))
            (bind _returnvalue (Utilities|Time|SetTimerbyEvent (EventDispatchers|CreateEvent :!event Custom|BeginExpireTell) (- (Variables|Default|GetDuration) (Variables|Default|GetExpirationTellDuration))))
            (Variables|Default|SetExpireTellDelayTimer _returnvalue)))
        (:then_1
          ;[1 1 1] Bind footstep and collision events
          (bind _asfort_player_pawn_athena (Utilities|Casting|CastToFortPlayerPawnAthena (Variables|Default|GetTargetPlayer)))
          (EventDispatchers|BindEventtoOnFootstepEvent _asfort_player_pawn_athena (EventDispatchers|CreateEvent :!event Custom|On%20Player%20Step))
          (bind _returnvalue_2 (AddComponent|NODE_AddFortCollisionAudioComponent-1))
          (Variables|Default|SetCollisionAudioComp _returnvalue_2)
          (Collision|BindEventtoOnActorHit _asfort_player_pawn_athena (EventDispatchers|CreateEvent :!event Custom|On%20Target%20Pawn%20Hit%20Event))
          (Audio|SetCollisionSoundatIndex (Variables|Default|GetCollisionAudioComp) 0 (Variables|Sound|GetSC_Collision))
          (EventDispatchers|BindEventtoOnPawnLanded _asfort_player_pawn_athena (EventDispatchers|CreateEvent :!event Custom|On%20Target%20Pawn%20Landed))
          (Character|BindEventtoMovementModeChangedDelegate _asfort_player_pawn_athena (EventDispatchers|CreateEvent :!event Custom|On%20Target%20Pawn%20Movement%20Mode%20Changed)))))))

(event Collision|EventHit
  (params (MyComp :type Primitive%20Component%20Object%20Reference) (Other :type Actor%20Object%20Reference) (OtherComp :type Primitive%20Component%20Object%20Reference) (bSelfMoved :type Boolean) (HitLocation :type Vector :default "0, 0, 0") (HitNormal :type Vector :default "0, 0, 0") (NormalImpulse :type Vector :default "0, 0, 0") (Hit :type Hit%20Result%20Structure%20%28by%20ref%29))
  (Development|PrintString MyComp))

(event EventOnRemoval(EitherVFX-preferthisversion)
  (params (MyTarget :type Actor%20Object%20Reference) (Parameters :type Gameplay%20Cue%20Parameters%20Structure%20%28by%20ref%29) (ParticleComponents :type Array%20of%20FXSystem%20Component%20Object%20References) (AudioComponents :type Array%20of%20Audio%20Component%20Object%20References) (BurstCameraShakeInstance :type Camera%20Shake%20Base%20Object%20Reference) (BurstDecalInstance :type Decal%20Actor%20Object%20Reference))
  (CallFunction|ClearExpireTell)
  (bind _asfort_player_pawn (Utilities|Casting|CastToFortPlayerPawn MyTarget))
  (Effects|SpawnFXSystemAttached :FXSystemAsset (Variables|Default|GetExpireFX) :AttachToComponent (Variables|/Script/Engine.Character|GetMesh _asfort_player_pawn) :AttachPointName "foot_l" :Scale "1.000000,1.000000,1.000000" :LocationType "SnapToTarget" :bAutoDestroy true :PoolingMethod "None" :bAutoActivate true)
  (Effects|SpawnFXSystemAttached :FXSystemAsset (Variables|Default|GetExpireFX) :AttachToComponent (Variables|/Script/Engine.Character|GetMesh _asfort_player_pawn) :AttachPointName "foot_r" :Scale "1.000000,1.000000,1.000000" :LocationType "SnapToTarget" :bAutoDestroy true :PoolingMethod "None" :bAutoActivate true)
  ;[1 1 1] Unbind footstep and collision events
  (bind _asfort_player_pawn_athena (Utilities|Casting|CastToFortPlayerPawnAthena MyTarget))
  (EventDispatchers|UnbindEventfromOnFootstepEvent _asfort_player_pawn_athena (EventDispatchers|CreateEvent :!event Custom|On%20Player%20Step))
  (Collision|UnbindEventfromOnActorHit _asfort_player_pawn_athena (EventDispatchers|CreateEvent :!event Custom|On%20Target%20Pawn%20Hit%20Event))
  (EventDispatchers|UnbindEventfromOnPawnLanded _asfort_player_pawn_athena (EventDispatchers|CreateEvent :!event Custom|On%20Target%20Pawn%20Landed))
  (Character|UnbindEventfromMovementModeChangedDelegate _asfort_player_pawn_athena (EventDispatchers|CreateEvent :!event Custom|On%20Target%20Pawn%20Movement%20Mode%20Changed))
  (Components|DestroyComponent (Variables|Default|GetCollisionAudioComp) self))

(event Custom|On%20Player%20Step
  (params (FootIndex :type Integer))
  ;[1 1 1] On player step, play this audio
  (Audio|PlaySoundatLocation (Variables|Default|GetTargetPlayer) (select (Pawn|IsLocallyControlled (Variables|Default|GetTargetPlayer)) (Variables|Sound|GetSC_Footstep_Local) (Variables|Sound|GetSC_Footstep_Remote)) (Transformation|GetActorLocation (Variables|Default|GetTargetPlayer))))

(event Custom|On%20Target%20Pawn%20Hit%20Event
  (params (SelfActor :type Actor%20Object%20Reference) (OtherActor :type Actor%20Object%20Reference) (NormalImpulse :type Vector) (Hit :type Hit%20Result%20Structure%20%28by%20ref%29))
  (bind (_bblockinghit _binitialoverlap _time _distance _location _impactpoint _normal _impactnormal) (Collision|BreakHitResult Hit))
  ;[1 1 1] On player hit event, play this audio
  (Audio|OnCollision (Variables|Default|GetCollisionAudioComp) _location (* _impactnormal (Transformation|GetVelocity (Variables|Default|GetTargetPlayer)))))

(event EventTick
  (params (DeltaSeconds :type Float%20%28single-precision%29))
  (bind _spawnedaudiocomps (Variables|Default|GetSpawnedAudioComps))
  (bind _output (Utilities|Array|Get(acopy) _spawnedaudiocomps))
  ;[1 1 1] Modulate slide volume based on velocity
  (if (> (Utilities|Array|Length _spawnedaudiocomps) 0)
    (Utilities|IsValid _output
      (:Is%20Valid
        (if (or (AI|Components|NavMovement|IsFalling (Variables|/Script/Engine.Character|GetCharacterMovement (Variables|Default|GetTargetPlayer))) (not (Utilities|Enum|Equal(Enum) (Variables|/Script/Engine.CharacterMovementComponent|GetMovementMode (Variables|/Script/Engine.Character|GetCharacterMovement (Variables|Default|GetTargetPlayer))) "MOVE_Walking")))
          (Audio|Parameter|SetFloatParameter _output "Speed" 0.0 :!class /Script/AudioExtensions.AudioParameterControllerInterface)
          (else
            (Audio|Parameter|SetFloatParameter _output "Speed" (Math|Float|MapRangeClamped (Math|Vector|VectorLength (Transformation|GetVelocity (Variables|Default|GetTargetPlayer))) 0.0 1e+03 0.0 1.0) :!class /Script/AudioExtensions.AudioParameterControllerInterface)))))))

(event Custom|On%20Target%20Pawn%20Landed
  (params (Hit :type Hit%20Result%20Structure%20%28by%20ref%29))
  ;[1 1 1] Landing sound
  (Audio|SpawnSoundAttached (Variables|Sound|GetSC_OnLand) (Variables|/Script/Engine.Actor|GetRootComponent (Variables|Default|GetTargetPlayer))))

(event Custom|On%20Target%20Pawn%20Movement%20Mode%20Changed
  (params (Character :type Character%20Object%20Reference) (PrevMovementMode :type EMovementMode%20Enum :default "MOVE_None") (PreviousCustomMode :type Byte))
  ;[1 1 1] Sound played when leaving ground
  (Utilities|IsValid Character
    (:Is%20Valid
      (switch Utilities|FlowControl|Switch|SwitchonEMovementMode (Variables|/Script/Engine.CharacterMovementComponent|GetMovementMode (Variables|/Script/Engine.Character|GetCharacterMovement Character))
        (:MOVE_Falling
          (Audio|SpawnSoundAttached (Variables|Sound|GetSC_OnJumpOrBeginFalling) (Variables|/Script/Engine.Actor|GetRootComponent (Variables|Default|GetTargetPlayer))))))))

(fn ConstructionScript)
